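# UIKit info box widget showing an icon, rich-text message and optional status indicator over a dark backplate, with cosmetic-gating placeholder content and decorator.
# PreConstruct applies exposed settings for visibility, icon size, material and alignment, and text wrap and justification.
# It shows preview text at design time or when flagged.
# Custom events let callers update text, icon visibility, size and material, and forward icon, type and configuration to the status indicator.
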WidgetBlueprint /FortUILibrary/UIKit/InfoBox/WBP_UIKit_InfoBox
graphs [EventGraph {nodes [
   n1 = SetText!() -> (OutputDelegate, then, InText:Text) {is=title K2Node_CustomEvent}
   > Set Text(Text=@n1.InText, self:Rich Text Block Object Reference=Get Text!() -> (Text:Common Rich Text Block Object Reference))
   n2 = SetIconSize!() -> (OutputDelegate, then, InSize:Vector 2D Structure) {is=title K2Node_CustomEvent}
   > Set IconSize(IconSize:Vector 2D Structure=@n2.InSize) -> (Output_Get:Vector 2D Structure)
   n3 = Get Icon() -> (Icon:Image Object Reference)
   n4 = Get!(self:Image Object Reference=@n3) -> (
    Brush_TintColor:Slate Color Structure
    Brush_DrawAs:ESlateBrushDrawType Enum
    Brush_Tiling:ESlateBrushTileType Enum
    Brush_Mirroring:ESlateBrushMirrorType Enum
    Brush_ImageSize:Vector 2D Structure
    Brush_Margin:Margin Structure
    Brush_ResourceObject:Object Reference
    Brush_OutlineSettings:Slate Brush Outline Settings Structure
   )
   > Set!(
    execute
    Brush_TintColor:Slate Color Structure=@n4
    Brush_DrawAs:ESlateBrushDrawType Enum=@n4.Brush_DrawAs
    Brush_Tiling:ESlateBrushTileType Enum=@n4.Brush_Tiling
    Brush_Mirroring:ESlateBrushMirrorType Enum=@n4.Brush_Mirroring
    Brush_ImageSize:Vector 2D Structure=Get IconSize() -> (IconSize:Vector 2D Structure)
    Brush_Margin:Margin Structure=@n4.Brush_Margin
    Brush_ResourceObject:Object Reference=@n4.Brush_ResourceObject
    Brush_OutlineSettings:Slate Brush Outline Settings Structure=@n4.Brush_OutlineSettings
    self:Image Object Reference=@n3
   ) -> (then, Output_Get:Slate Brush Structure)
   n5 = SetShowIcon!() -> (OutputDelegate, then, Show Icon:bool=~) {is=title K2Node_CustomEvent}
   > n6 = Set ShowIcon(ShowIcon:bool=@n5.Show Icon) -> (Output_Get:bool)
   > Set Visibility(Visibility=Select(Option 0=Collapsed, Option 1=SelfHitTestInvisible, Index=@n6), self=Get Icon())
   n7 = SetIconImage!() -> (OutputDelegate, then, Material:Material Interface Object Reference) {is=title K2Node_CustomEvent}
   > SetBrushFromMaterial(self=Get Icon(), Material=@n7.Material)
   n8 = SetStatusIndicator!() -> (
    OutputDelegate
    then
    Icon:Object Reference
    StatusType:E_UI_StatusIndicator_Type Enum=NewEnumerator0
    Config:E_UI_StatusIndicator_Configuration Enum=NewEnumerator0
   ) {is=title K2Node_CustomEvent}
   n9 = Get WBP_UIKit_StatusIndicator() -> (WBP_UIKit_StatusIndicator:WBP UIKit Status Indicator Object Reference)
   > SetIcon!(execute, self:WBP UIKit Status Indicator Base Object Reference=@n9, IconAsset:Object Reference=@n8.Icon) -> (then)
   > SetIndicatorType!(execute, self:WBP UIKit Status Indicator Base Object Reference=@n9, IndicatorType:E_UI_StatusIndicator_Type Enum=@n8.StatusType) -> (then)
   > SetIndicatorConfiguration!(
    execute
    self:WBP UIKit Status Indicator Base Object Reference=@n9
    InCount:int=0
    IndicatorConfig:E_UI_StatusIndicator_Configuration Enum=@n8.Config
   ) -> (then)
   n10 = Event PreConstruct()
   > n11 = Sequence!(execute) -> (then_0, then_1, then_2:Exec)
   Set Visibility(
    execute=@n11.then_1
    Visibility=Select(Option 0=Collapsed, Option 1=SelfHitTestInvisible, Index=Get ShowIcon() -> (ShowIcon:bool))
    self=Get Icon()
   )
   > Set Visibility(Visibility=Select(Option 0=Collapsed, Option 1=SelfHitTestInvisible, Index=Get ShowText() -> (ShowText:bool)), self=Get Text())
   > Set Visibility(
    Visibility=Select(Option 0=Collapsed, Option 1=SelfHitTestInvisible, Index=Get Show Background() -> (Show Background:bool))
    self=Get WBP_UIKit_Backplate() -> (WBP_UIKit_Backplate:WBP UIKit Backplate Object Reference)
   )
   > Set Visibility(
    Visibility=Select(Option 0=Collapsed, Option 1=SelfHitTestInvisible, Index=Get ShowStatusIndicator() -> (ShowStatusIndicator:bool))
    self=Get WBP_UIKit_StatusIndicator()
   )
   > SetIconSize!(execute, self, InSize:Vector 2D Structure=Get IconSize()) -> (then) {is=title K2Node_CallFunction}
   > SetAutoWrapText!(execute, self:Rich Text Block Object Reference=Get Text(), InAutoTextWrap:bool=Get AutoWrapText() -> (AutoWrapText:bool)) -> (then)
   > Set Justification!(
    execute
    Justification:ETextJustify Enum=Get TextJustification() -> (TextJustification:ETextJustify Enum=Left)
    self:Text Layout Widget Object Reference=Get Text()
   ) -> (then, Output_Get:ETextJustify Enum=Left)
   Set VerticalAlignment!(
    execute=@n11.then_2
    VerticalAlignment:EVerticalAlignment Enum=Select(
     Option 0:EVerticalAlignment Enum=VAlign_Center
     Option 1:EVerticalAlignment Enum=VAlign_Top
     Index=Get IconTopAlignment() -> (IconTopAlignment:bool)
    ) -> (ReturnValue:EVerticalAlignment Enum)
    self:Overlay Slot Object Reference=SlotAsOverlaySlot(Widget=Get Icon())
   ) -> (then, Output_Get:EVerticalAlignment Enum=VAlign_Fill)
   n12 = Branch(execute=@n11.then_0, Condition=OR Boolean(A=@n10.IsDesignTime, B=Get Use Preview Text() -> (Use Preview Text:bool)))
   > n13 = SetText(self:Self Object Reference, InText:Text=Get PreviewText() -> (PreviewText:Text)) {is=title K2Node_CallFunction}
   n14 = Get IconMaterial() -> (IconMaterial:Material Interface Object Reference)
   n15 = Branch(execute=[@n13.then, @n12.else], Condition=IsNullItem!(ItemObject:Object Reference=@n14) -> (Is Null Item:bool))
   SetIconImage!(execute=@n15.else, self, Material:Material Interface Object Reference=@n14) -> (then) {is=title K2Node_CallFunction}
  ]}]
variables [
 AutoWrapText {type=boolean, default=true}
 IconMaterial {type=MaterialInterface, default=null}
 IconSize {type=Vector2D, default=<32 32>}
 IconTopAlignment {type=boolean}
 PreviewText {type=string, default="Items with the <img id=\"WarningIcon\"/> symbol can only be used in experiences rated Teen."}
 Show Background {type=boolean, default=true}
 ShowIcon {type=boolean, default=true}
 ShowStatusIndicator {type=boolean}
 ShowText {type=boolean, default=true}
 TextJustification {type=ETextJustify, default=Left}
 Use Preview Text {type=boolean}
]
widgets {
 info {inheritedWidgetCount=0, parentClass=&UserWidget, rootWidgetClass=&Overlay, widgetCount=7}
 widgets [
  Overlay_0 {widgetClassPath=&Overlay, widget=Overlay_0}
  WBP_UIKit_Backplate {
   widgetClassPath=&WBP_UIKit_Backplate
   bIsVariable=true
   parent=Overlay_0
   widget=WBP_UIKit_Backplate
   properties {brightness=Dark}
   slot {class=OverlaySlot, properties {horizontalAlignment=HAlign_Fill, verticalAlignment=VAlign_Fill}}
  }
  HorizontalBox {
   widgetClassPath=&HorizontalBox
   bIsVariable=true
   parent=Overlay_0
   widget=HorizontalBox
   slot {class=OverlaySlot, properties {horizontalAlignment=HAlign_Fill, verticalAlignment=VAlign_Fill, padding=<ltrb 16 8 16 8>}}
  }
  Overlay_1 {
   widgetClassPath=&Overlay
   parent=HorizontalBox
   widget=Overlay_1
   slot {class=HorizontalBoxSlot, properties {horizontalAlignment=HAlign_Left, padding=<ltrb 0 0 10 0>}}
  }
  Icon {
   widgetClassPath=&Image
   bIsVariable=true
   parent=Overlay_1
   widget=Icon
   properties {
    visibility=SelfHitTestInvisible
    brush {
     bIsDynamicallyLoaded=false
     drawAs=Image
     imageType=NoImage
     mirroring=NoMirror
     resourceName=None
     resourceObject=/Game/Athena/UI/Frontend/CosmeticGating/Materials/MI_UI_CosmeticGate_WarningIcon
     tiling=NoTile
     imageSize=<34359746560 0>
     margin=<ltrb 0 0 0 0>
     outlineSettings {
      bUseBrushTransparency=false
      roundingType=HalfHeightRadius
      width=0
      color {colorUseRule=UseColor_Specified, specifiedColor=<rgba 0 0 0 0>}
      cornerRadii=<0 0 0 0>
     }
     tintColor {colorUseRule=UseColor_Specified, specifiedColor=<rgba 1 1 1 1>}
     uVRegion {bIsValid=false, max=<0 0>, min=<0 0>}
    }
   }
   slot {class=OverlaySlot}
  }
  Text {
   widgetClassPath=&CommonRichTextBlock
   bIsVariable=true
   parent=HorizontalBox
   widget=Text
   properties {
    autoWrapText=true
    bOverrideDefaultStyle=true
    mobileTextBlockScale=1.3
    text="[PH] Your selected outfit can't be worn in experiences of this rating."
    decoratorClasses [/Game/Athena/UI/Frontend/CosmeticGating/Data/RichTextDecorator_CosmeticGate.RichTextDecorator_CosmeticGate_C]
    defaultTextStyleOverride {
     overflowPolicy=Clip
     transformPolicy=None
     colorAndOpacity {colorUseRule=UseColor_Specified, specifiedColor=<rgba 1 1 1 0.75>}
     font {
      bForceMonospaced=false
      bMaterialIsStencil=false
      fontMaterial=null
      fontObject=&HeadingNow
      letterSpacing=10
      monospacedWidth=1
      size=14.25
      skewAmount=0
      typefaceFontName=74 Regular
      outlineSettings {
       bApplyOutlineToDropShadows=false
       bMiteredCorners=false
       bSeparateFillAlpha=false
       outlineMaterial=null
       outlineSize=0
       outlineColor=<rgba 0 0 0 1>
      }
     }
     highlightColor {colorUseRule=UseColor_Specified, specifiedColor=<rgba 0 0 0 1>}
     highlightShape {
      bIsDynamicallyLoaded=false
      drawAs=Image
      imageType=NoImage
      mirroring=NoMirror
      resourceName=None
      resourceObject=null
      tiling=NoTile
      imageSize=<32 32>
      margin=<ltrb 0 0 0 0>
      outlineSettings {
       bUseBrushTransparency=false
       roundingType=HalfHeightRadius
       width=0
       color {colorUseRule=UseColor_Specified, specifiedColor=<rgba 0 0 0 0>}
       cornerRadii=<0 0 0 0>
      }
      tintColor {colorUseRule=UseColor_Specified, specifiedColor=<rgba 1 1 1 1>}
      uVRegion {bIsValid=false, max=<0 0>, min=<0 0>}
     }
     selectedBackgroundColor {colorUseRule=UseColor_Foreground, specifiedColor=<rgba 1 0 1 1>}
     shadowColorAndOpacity=<rgba 0 0 0 1>
     shadowOffset=<0 0>
     strikeBrush {
      bIsDynamicallyLoaded=false
      drawAs=Image
      imageType=NoImage
      mirroring=NoMirror
      resourceName=None
      resourceObject=null
      tiling=NoTile
      imageSize=<32 32>
      margin=<ltrb 0 0 0 0>
      outlineSettings {
       bUseBrushTransparency=false
       roundingType=HalfHeightRadius
       width=0
       color {colorUseRule=UseColor_Specified, specifiedColor=<rgba 0 0 0 0>}
       cornerRadii=<0 0 0 0>
      }
      tintColor {colorUseRule=UseColor_Specified, specifiedColor=<rgba 1 1 1 1>}
      uVRegion {bIsValid=false, max=<0 0>, min=<0 0>}
     }
     underlineBrush {
      bIsDynamicallyLoaded=false
      drawAs=Image
      imageType=NoImage
      mirroring=NoMirror
      resourceName=None
      resourceObject=null
      tiling=NoTile
      imageSize=<32 32>
      margin=<ltrb 0 0 0 0>
      outlineSettings {
       bUseBrushTransparency=false
       roundingType=HalfHeightRadius
       width=0
       color {colorUseRule=UseColor_Specified, specifiedColor=<rgba 0 0 0 0>}
       cornerRadii=<0 0 0 0>
      }
      tintColor {colorUseRule=UseColor_Specified, specifiedColor=<rgba 1 1 1 1>}
      uVRegion {bIsValid=false, max=<0 0>, min=<0 0>}
     }
    }
   }
   slot {class=HorizontalBoxSlot, properties {verticalAlignment=VAlign_Center, size {sizeRule=Fill, value=1}}}
  }
  WBP_UIKit_StatusIndicator {
   widgetClassPath=/FortUILibrary/UIKit/StatusIndicator/WBP_UIKit_StatusIndicator
   bIsVariable=true
   parent=HorizontalBox
   widget=WBP_UIKit_StatusIndicator
   properties {type=Attention, visibility=Collapsed}
   slot {class=HorizontalBoxSlot, properties {padding=<ltrb 16 0 0 0>}}
  }
 ]
}
